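# Niagara effect for the IceWater mic stand cosmetic, shedding cold smoke, snowflakes and ice shards.
# Each emitter uses particle partitions to spawn at several heights along the stand.
# Smoke and snowflakes sink with frame-rate-compensated random sideways drift, while mesh shards fall under gravity.
# Smoke color and alpha follow time of day, with alpha also varying by User.IsFrontEnd; snowflakes use two sprite renderers and compute a corona size.
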NiagaraSystem /SparksCosmetics/Instrument/MicStand/IceWater/Effects/NS_MicStand_IceWater
dependencies {
 usedDataInterfaces [
  &NiagaraDataInterfaceArrayFloat
  &NiagaraDataInterfaceArrayFloat3
  &NiagaraDataInterfaceCurve
  &NiagaraDataInterfaceArrayColor
 ]
 usedDynamicInputs []
 usedModules [&SystemState, &Timeline]
}
emitters {
 IceLightSmoke {
  emitterUpdateScript [
   &EmitterState(Life Cycle Mode=::NewEnumerator1, Loop Duration=0.1, UseLoopDelay=false)
   &SpawnBurst_Instantaneous(Spawn Count=5, Use Spawn Probability=false, UseLoopCountLimit=true)
   &SpawnRate(Spawn Probability=0.8, SpawnRate=14, Use Spawn Probability=true)
  ]
  particleSpawnScript [
   &PartitionParticles(Number Of Partitions=4)
   &InitializeParticle(
    Color=<rgba 0.6562 0.8155 1 0.2>
    Lifetime Max=&SelectFloatFromArray(Direct Array Index=$Particles.PartitionParticles.Partition, Float Selection Array=DataInterface() {propertyValues {FloatData [4,4.5,4.5]}})
    Lifetime Min:NiagaraFloat=&SelectFloatFromArray(Direct Array Index=$Particles.PartitionParticles.Partition, Float Selection Array=DataInterface() {propertyValues {FloatData [3,3,2]}})
    Lifetime Mode=::NewEnumerator1
    Mass Mode=::NewEnumerator3
    Position Offset=&SelectVectorFromArray(
     Direct Array Index=$Particles.PartitionParticles.Partition
     Vector Selection Array=DataInterface() {propertyValues {FloatData [<0 0 -47>, <0 0 60>, <0 0 80>, <0 0 -5>]}}
    )
    Sprite Rotation Mode=::NewEnumerator2
    Sprite Size Mode=::NewEnumerator1
    Uniform Sprite Size Max=&SelectFloatFromArray(Direct Array Index=$Particles.PartitionParticles.Partition, Float Selection Array=DataInterface() {propertyValues {FloatData [75,40,40]}})
    Uniform Sprite Size Min=&SelectFloatFromArray(Direct Array Index=$Particles.PartitionParticles.Partition, Float Selection Array=DataInterface() {propertyValues {FloatData [50,20,20]}})
    UsePositionOffset=true
    Write Material Random=true
   )
   &ShapeLocation(
    MaskSpawnGroup=false
    Non Uniform Scale=<1 1 0.25>
    Override Seed=false
    Sphere Radius=&SelectFloatFromArray(Direct Array Index=$Particles.PartitionParticles.Partition, Float Selection Array=DataInterface() {propertyValues {FloatData [13,0,0]}})
    UseOwnerScale=true
   )
   &AddVelocity(
    Cone Angle=32
    Cone Axis=<0 0 1>
    Distribution Along Cone Axis=0.25
    Module.UseConeFalloff=true
    Module.UseVelDistribution=true
    Override Seed=false
    Speed Falloff From Cone Axis=0.5
    Velocity Mode=::NewEnumerator2
    Velocity Speed=&RandomRangeFloat(Minimum=500, Maximum=850)
   ) {enabled=false}
  ]
  particleUpdateScript [
   &ParticleState()
   /Niagara/Modules/Update/SubUV/SubUVAnimation(Number Of Frames=63)
   &ScaleSpriteSize(
    Scale Sprite Size Mode=::NewEnumerator0
    Uniform Scale Factor=&FloatFromCurve(
     FloatCurve=DataInterface() {propertyValues {Curve {keys [
         {arriveTangent=-0.0977, interpMode=RCIM_Cubic, leaveTangent=-0.0977, tangentMode=RCTM_User, time=0, value=0.25}
         {arriveTangent=0.7151, interpMode=RCIM_Cubic, leaveTangent=0.7151, time=0.3, value=0.3}
         <at 1.0489 1 RCIM_Cubic>
        ]}}}
     Scale Curve=0.75
    )
   )
   &ScaleColor(
    Scale Alpha=&FloatFromCurve(
     FloatCurve=DataInterface() {propertyValues {Curve {keys [
         <at 0 0 RCIM_Cubic>
         {arriveTangent=0.257, interpMode=RCIM_Cubic, leaveTangent=0.257, tangentMode=RCTM_User, time=0.2, value=0.65}
         {arriveTangent=0.1612, interpMode=RCIM_Cubic, leaveTangent=0.1612, tangentMode=RCTM_User, time=1.0011, value=0}
        ]}}}
     Scale Curve=0.5
    )
    ScaleA=true
    ScaleRGB=true
   ) {enabled=false}
   /SetVariables(Particles.Position=&AddVectorToPosition(
     Position=$Particles.Position
     Vector=&MakeVector(
      X=&RandomRangeFloat(
       Minimum=&FloatFromCurve(
        FloatCurve=DataInterface() {propertyValues {Curve {keys [
            <at 0 0 RCIM_Cubic>
            {arriveTangent=0.0833, interpMode=RCIM_Cubic, leaveTangent=0.0833, time=0.3, value=0.1}
            {arriveTangent=1.2857, interpMode=RCIM_Cubic, leaveTangent=1.2857, time=0.6, value=0.05}
            {arriveTangent=3.401, tangentMode=RCTM_User, time=1, value=1}
           ]}}}
        Scale Curve=&Multiply_Float(A=-0.15, B=&Multiply_Float(A=60, B=$Engine.DeltaTime))
       )
       Maximum=&FloatFromCurve(
        FloatCurve=DataInterface() {propertyValues {Curve {keys [
            <at 0 0 RCIM_Cubic>
            {arriveTangent=0.8571, interpMode=RCIM_Cubic, leaveTangent=0.8571, time=0.4, value=-0.05}
            {arriveTangent=1.75, interpMode=RCIM_Cubic, leaveTangent=1.75, time=0.7, value=0.6}
            {arriveTangent=3.0412, tangentMode=RCTM_User, time=1, value=1}
           ]}}}
        Scale Curve=&Multiply_Float(A=0.15, B=&Multiply_Float(A=60, B=$Engine.DeltaTime))
       )
      )
      Y=0
      Z=&FloatFromCurve(
       FloatCurve=DataInterface() {propertyValues {Curve {keys [<at 0 1 RCIM_Cubic>, {tangentMode=RCTM_User, time=1, value=1.25}]}}}
       Scale Curve=&Multiply_Float(
        A=&SelectFloatFromArray(
         Direct Array Index=$Particles.Initial.PartitionParticles.Partition
         Float Selection Array=DataInterface() {propertyValues {FloatData [-0.075,-0.25,-0.25,-0.25]}}
        )
        B=&Multiply_Float(A=60, B=$Engine.DeltaTime)
       )
      )
     )
    )) {isSetParameters=true, module=SetVariables_#1}
   &GravityForce(Gravity=&SelectVectorFromArray(
     Direct Array Index=$Particles.PartitionParticles.Partition
     Vector Selection Array=DataInterface() {propertyValues {FloatData [<0 0 -3>, <0 0 0>, <0 0 -0.5>]}}
    )) {enabled=false}
   &SolveForcesAndVelocity(Clamp Velocity=true, Limit Acceleration=true)
   &CameraOffset(Camera Offset Amount=&SelectFloatFromArray(
     Direct Array Index=$Particles.Initial.PartitionParticles.Partition
     Float Selection Array=DataInterface() {propertyValues {FloatData [25,18,18]}}
    ))
   &ScaleColor(
    Scale Alpha=&Multiply_Float(
     A=&FloatFromCurve(FloatCurve=DataInterface() {propertyValues {Curve {keys [
          <at 0 0 RCIM_Cubic>
          {arriveTangent=0.257, interpMode=RCIM_Cubic, leaveTangent=0.257, tangentMode=RCTM_User, time=0.3, value=0.45}
          {arriveTangent=0.1612, interpMode=RCIM_Cubic, leaveTangent=0.1612, tangentMode=RCTM_User, time=1.0011, value=0}
         ]}}})
     B=&FNTimeOfDayScalar(Night Scale=1, Morning Scale=0.7, Day Scale=0.6, Evening Scale=0.7)
    )
    Scale RGB=&FN_Select_Vector_from_Time_Of_Day(Morning=<0.9 0.9 0.9>, Day=<1 1 1>, Evening=<1 1 1>, Night=<0.5 0.5 1>)
    ScaleA=true
    ScaleRGB=true
   ) {module=ScaleColor002}
   &ScaleColor(
    Scale Alpha=&MakeCustomFloatFromBool(Bool=$User.IsFrontEnd, True Float=0.5, False Float=0.55)
    ScaleA=true
    ScaleRGB=false
   ) {module=ScaleColor001}
  ]
  rendererClasses [&NiagaraSpriteRendererProperties]
  renderers [{rendererClass=&NiagaraSpriteRendererProperties}]
 }
 IceShards {
  emitterUpdateScript [
   &EmitterState(Life Cycle Mode=::NewEnumerator1, Loop Duration=0.1, UseLoopDelay=false)
   &SpawnRate(Spawn Probability=0.45, SpawnRate=5, Use Spawn Probability=true)
  ]
  particleSpawnScript [
   &PartitionParticles(Number Of Partitions=5)
   &InitializeParticle(
    Color=&SelectLinearColorFromArray(
     Array Sampling Mode=::NewEnumerator0
     Direct Array Index=?
     Color Selection Array=DataInterface() {propertyValues {ColorData [<rgba 1.2917 1.9252 2 1>, <rgba 3.6132 4.2075 5 1>]}}
     Random Seed=0
    )
    Lifetime Max=2.5
    Lifetime Min=0.85
    Lifetime Mode=::NewEnumerator1
    Mass Mode=::NewEnumerator3
    Mesh Scale Max=<0.75 0.75 0.75>
    Mesh Scale Min=<0.45 0.45 0.45>
    Mesh Scale Mode=::NewEnumerator2
    Position Offset=&SelectVectorFromArray(
     Direct Array Index=$Particles.PartitionParticles.Partition
     Vector Selection Array=DataInterface() {propertyValues {FloatData [<0 0 -45>, <0 0 40>, <0 0 80>, <0 0 -10>, <0 0 10>]}}
    )
    Sprite Rotation Mode=::NewEnumerator2
    Sprite Size Mode=::NewEnumerator1
    UsePositionOffset=true
    Write Material Random=true
   )
   &ShapeLocation(
    MaskSpawnGroup=false
    Non Uniform Scale=<1 0.1 1>
    Override Seed=false
    Sphere Radius=&SelectFloatFromArray(Direct Array Index=$Particles.PartitionParticles.Partition, Float Selection Array=DataInterface() {propertyValues {FloatData [8,5,5,5,5]}})
    UseOwnerScale=true
   )
   /SetVariables(Particles.Rotation:NiagaraFloat=&RandomRangeFloat(Minimum=-80, Maximum=80)) {isSetParameters=true, module=SetVariables_#2}
   &AddVelocity(
    Cone Angle=32
    Cone Axis=<0 0 1>
    Distribution Along Cone Axis=0.25
    Module.UseConeFalloff=true
    Module.UseVelDistribution=true
    Override Seed=false
    Speed Falloff From Cone Axis=0.5
    Velocity Mode=::NewEnumerator2
    Velocity Speed=&RandomRangeFloat(Minimum=500, Maximum=850)
   ) {enabled=false}
   &InitialMeshOrientation(Use Rotation Vector=false)
  ]
  particleUpdateScript [
   &ParticleState()
   &UpdateMeshOrientation(Rotation Vector=&RandomRangeVector(Minimum=<0 0 0>, Maximum=<1 1 1>))
   &ScaleMeshSize(Scale Factor=&VectorFromFloat(Value=&FloatFromCurve(
      FloatCurve=DataInterface() {propertyValues {Curve {keys [
          {arriveTangent=9.0725, interpMode=RCIM_Cubic, leaveTangent=9.0725, tangentMode=RCTM_User, time=0, value=0}
          {arriveTangent=-2.5249, interpMode=RCIM_Cubic, leaveTangent=-2.5249, tangentMode=RCTM_User, time=0.1408, value=0.9847}
          {arriveTangent=-0.0405, interpMode=RCIM_Cubic, leaveTangent=-0.0405, tangentMode=RCTM_User, time=1, value=0}
         ]}}}
      Scale Curve=0.45
     )))
   &ScaleColor(
    Scale Alpha=&FloatFromCurve(
     FloatCurve=DataInterface() {propertyValues {Curve {keys [
         <at 0 0 RCIM_Cubic>
         {arriveTangent=0.257, interpMode=RCIM_Cubic, leaveTangent=0.257, tangentMode=RCTM_User, time=0.2, value=0.95}
         {arriveTangent=0.1612, interpMode=RCIM_Cubic, leaveTangent=0.1612, tangentMode=RCTM_User, time=1.0011, value=0}
        ]}}}
     Scale Curve=0.5
    )
    ScaleA=true
    ScaleRGB=true
   )
   /SetVariables(Particles.Position=&AddVectorToPosition(
     Position=$Particles.Position
     Vector=&MakeVector(
      X=&RandomRangeFloat(
       Minimum=&FloatFromCurve(
        FloatCurve=DataInterface() {propertyValues {Curve {keys [
            <at 0 0 RCIM_Cubic>
            {arriveTangent=0.3333, interpMode=RCIM_Cubic, leaveTangent=0.3333, time=0.3, value=-0.15}
            {arriveTangent=1.6429, interpMode=RCIM_Cubic, leaveTangent=1.6429, time=0.6, value=0.2}
            {arriveTangent=3.401, tangentMode=RCTM_User, time=1, value=1}
           ]}}}
        Scale Curve=-1
       )
       Maximum=&FloatFromCurve(FloatCurve=DataInterface() {propertyValues {Curve {keys [
            <at 0 0 RCIM_Cubic>
            {arriveTangent=-0.7798, interpMode=RCIM_Cubic, leaveTangent=-0.7798, tangentMode=RCTM_User, time=0.4, value=0.175}
            {arriveTangent=1.375, interpMode=RCIM_Cubic, leaveTangent=1.375, time=0.7, value=0.75}
            {arriveTangent=3.0412, tangentMode=RCTM_User, time=1, value=1}
           ]}}})
      )
      Y=0
      Z=&FloatFromCurve(
       FloatCurve=DataInterface() {propertyValues {Curve {keys [<at 0 1.0625 RCIM_Cubic>, {tangentMode=RCTM_User, time=1, value=1.25}]}}}
       Scale Curve=-0.5
      )
     )
    )) {enabled=false, isSetParameters=true, module=SetVariables_#1}
   &GravityForce(Gravity=<0 0 -85>)
   &SolveForcesAndVelocity(Clamp Velocity=true, Limit Acceleration=true)
   &CameraOffset(Camera Offset Amount=8)
   &DynamicMaterialParameters(
    Index 0 Param 1=0.35
    Module.Index0Param0Write=true
    Module.Index0Param1Write=true
    Module.Index0Param2Write=true
    Module.Index0Param3Write=true
   )
  ]
  rendererClasses [&NiagaraMeshRendererProperties]
  renderers [{rendererClass=&NiagaraMeshRendererProperties, rendererIndex=0}]
 }
 SnowFlakes {
  emitterUpdateScript [
   &EmitterState(Life Cycle Mode=::NewEnumerator1, Loop Duration=0.1, UseLoopDelay=false)
   &SpawnBurst_Instantaneous(Spawn Count=7, Use Spawn Probability=false, UseLoopCountLimit=true)
   &SpawnRate(Spawn Probability=0.9, SpawnRate=15, Use Spawn Probability=true)
  ]
  particleSpawnScript [
   &PartitionParticles(Number Of Partitions=3)
   &InitializeParticle(
    Color=&SelectLinearColorFromArray(
     Array Sampling Mode=::NewEnumerator0
     Direct Array Index=?
     Color Selection Array=DataInterface() {propertyValues {ColorData [<rgba 1.2917 1.9252 2 1>, <rgba 3.6132 4.2075 5 1>]}}
     Random Seed=0
    )
    Lifetime Max=&SelectFloatFromArray(Direct Array Index=$Particles.PartitionParticles.Partition, Float Selection Array=DataInterface() {propertyValues {FloatData [3,2.25,2.25]}})
    Lifetime Min:NiagaraFloat=&SelectFloatFromArray(Direct Array Index=$Particles.PartitionParticles.Partition, Float Selection Array=DataInterface() {propertyValues {FloatData [2,1.25,1.25]}})
    Lifetime Mode=::NewEnumerator1
    Mass Mode=::NewEnumerator3
    Position Offset=&SelectVectorFromArray(
     Direct Array Index=$Particles.PartitionParticles.Partition
     Vector Selection Array=DataInterface() {propertyValues {FloatData [<0 0 -10>, <0 0 40>, <0 0 80>]}}
    )
    Sprite Rotation Mode=::NewEnumerator2
    Sprite Size Mode=::NewEnumerator1
    UsePositionOffset=true
    Write Material Random=true
   )
   &ShapeLocation(
    MaskSpawnGroup=false
    Override Seed=false
    Sphere Radius=&SelectFloatFromArray(Direct Array Index=$Particles.PartitionParticles.Partition, Float Selection Array=DataInterface() {propertyValues {FloatData [5,5,5]}})
    UseOwnerScale=true
   )
   /SetVariables(Particles.Rotation:NiagaraFloat=&RandomRangeFloat(Minimum=-80, Maximum=80)) {isSetParameters=true, module=SetVariables_#2}
   &AddVelocity(
    Cone Angle=32
    Cone Axis=<0 0 1>
    Distribution Along Cone Axis=0.25
    Module.UseConeFalloff=true
    Module.UseVelDistribution=true
    Override Seed=false
    Speed Falloff From Cone Axis=0.5
    Velocity Mode=::NewEnumerator2
    Velocity Speed=&RandomRangeFloat(Minimum=500, Maximum=850)
   ) {enabled=false}
  ]
  particleUpdateScript [
   &ParticleState()
   /SetVariables(Particles.CoronaSpriteSize=&Multiply_Vector2D_ByFloat(
     Float=&FloatFromCurve(FloatCurve=DataInterface() {propertyValues {Curve {keys [
          {interpMode=RCIM_Cubic, tangentMode=RCTM_User, time=0, value=0}
          {interpMode=RCIM_Cubic, tangentMode=RCTM_User, time=0.5, value=1}
          {tangentMode=RCTM_User, time=1, value=0}
         ]}}})
     Vector2D=&RandomRangeVector2D(Minimum=<3 3>, Maximum=<4.5 4.5>)
    )) {isSetParameters=true}
   &ScaleSpriteSize(
    Scale Sprite Size Mode=::NewEnumerator0
    Uniform Scale Factor=&FloatFromCurve(
     FloatCurve=DataInterface() {propertyValues {Curve {keys [
         {arriveTangent=-0.0977, interpMode=RCIM_Cubic, leaveTangent=-0.0977, tangentMode=RCTM_User, time=0, value=0}
         <at 0.5 0.45 RCIM_Cubic>
         <at 1.0489 0 RCIM_Cubic>
        ]}}}
     Scale Curve=0.5
    )
   )
   &ScaleColor(
    Scale Alpha=&FloatFromCurve(
     FloatCurve=DataInterface() {propertyValues {Curve {keys [
         <at 0 0 RCIM_Cubic>
         {arriveTangent=0.257, interpMode=RCIM_Cubic, leaveTangent=0.257, tangentMode=RCTM_User, time=0.2, value=0.95}
         {arriveTangent=0.1612, interpMode=RCIM_Cubic, leaveTangent=0.1612, tangentMode=RCTM_User, time=1.0011, value=0}
        ]}}}
     Scale Curve=0.5
    )
    ScaleA=true
    ScaleRGB=true
   )
   /SetVariables(Particles.Position=&AddVectorToPosition(
     Position=$Particles.Position
     Vector=&MakeVector(
      X=&RandomRangeFloat(
       Minimum=&FloatFromCurve(
        FloatCurve=DataInterface() {propertyValues {Curve {keys [
            <at 0 0 RCIM_Cubic>
            {arriveTangent=0.3333, interpMode=RCIM_Cubic, leaveTangent=0.3333, time=0.3, value=-0.15}
            {arriveTangent=1.6429, interpMode=RCIM_Cubic, leaveTangent=1.6429, time=0.6, value=0.2}
            {arriveTangent=3.401, tangentMode=RCTM_User, time=1, value=1}
           ]}}}
        Scale Curve=&Multiply_Float(A=-1, B=&Multiply_Float(A=60, B=$Engine.DeltaTime))
       )
       Maximum=&FloatFromCurve(
        FloatCurve=DataInterface() {propertyValues {Curve {keys [
            <at 0 0 RCIM_Cubic>
            {arriveTangent=-0.7798, interpMode=RCIM_Cubic, leaveTangent=-0.7798, tangentMode=RCTM_User, time=0.4, value=0.175}
            {arriveTangent=1.375, interpMode=RCIM_Cubic, leaveTangent=1.375, time=0.7, value=0.75}
            {arriveTangent=3.0412, tangentMode=RCTM_User, time=1, value=1}
           ]}}}
        Scale Curve=&Multiply_Float(A=1, B=&Multiply_Float(A=60, B=$Engine.DeltaTime))
       )
      )
      Y=0
      Z=&FloatFromCurve(
       FloatCurve=DataInterface() {propertyValues {Curve {keys [<at 0 1 RCIM_Cubic>, {tangentMode=RCTM_User, time=1, value=1.25}]}}}
       Scale Curve=&Multiply_Float(A=-0.3, B=&Multiply_Float(A=60, B=$Engine.DeltaTime))
      )
     )
    )) {isSetParameters=true, module=SetVariables_#1}
   &GravityForce(Gravity=<0 0 -2>)
   &SolveForcesAndVelocity(Clamp Velocity=true, Limit Acceleration=true)
   &CameraOffset(Camera Offset Amount=8)
  ]
  rendererClasses [&NiagaraSpriteRendererProperties]
  renderers [
   {rendererClass=&NiagaraSpriteRendererProperties}
   {rendererClass=&NiagaraSpriteRendererProperties, rendererIndex=1}
  ]
 }
}
summary {
 systemName=NS_MicStand_IceWater
 emitters [
  {emitterName=IceLightSmoke, rendererClasses [&NiagaraSpriteRendererProperties]}
  {emitterName=SnowFlakes, rendererClasses [&NiagaraSpriteRendererProperties]}
  {emitterName=IceShards, rendererClasses [&NiagaraMeshRendererProperties]}
 ]
 userVariables [User.IsFrontEnd {type=NiagaraBool}]
}
